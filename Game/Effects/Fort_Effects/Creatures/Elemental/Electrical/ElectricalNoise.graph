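# Additive unlit particle sprite material for electrical elemental creature effects, combining particle-color emissive with a noisy refraction strength.
# A tiling noise normal map, panned by particle relative time, slightly warps the UVs of a dust cloud texture.
# The squared cloud value times particle alpha blends toward the Max Refraction parameter, with depth fade on all but the mobile path.
# Particle RGB runs separately through the shared FX emissive adjustments function.
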Material /Game/Effects/Fort_Effects/Creatures/Elemental/Electrical/ElectricalNoise
graph {
 properties {
  bUsedWithParticleSprites=true
  blendMode=BLEND_Additive
  pixelDepthOffsetMode=PDOM_Legacy
  refractionMethod=RM_IndexOfRefraction
  shadingModel=MSM_Unlit
  translucencyPass=MTP_BeforeDOF
  usedShadingModels=Unlit
 }
 nodes [
  n1 = ParticleColor()
  n2 = TextureCoordinate() {uTiling=0.25, vTiling=0.25}
  n3 = ParticleRelativeTime()
  n4 = ComponentMask(None=TextureSample(UVs=Panner(Coordinate=@n2, Time=@n3) {speedX=0.125, speedY=0.0625}) {samplerType=SAMPLERTYPE_Normal, texture=/Game/VisualThreatManager/StormVisuals/Textures/TilingNoise06_N}.RGB) {g=true, r=true}
  n5 = TextureSample(UVs=Add(A=Multiply(A=@n4) {constB=0.1}, B=TextureCoordinate())) {texture=/Game/Packages/Fortress_Transylvania/Haunted_Forest/Swamps/Textures/T_dust_cloud_01}
  n6 = Multiply(A=@n1.A, B=Multiply(A=@n5.R, B=@n5.R))
  n7 = DepthFade(Opacity=@n6)
  LinearInterpolate(
   B=ScalarParameter() {defaultValue=1.5, parameterName=Max Refraction}
   Alpha=ShadingPathSwitch(Default=@n7, Deferred=@n7, Mobile=@n6) {desc=Migrated from FeatureLevelSwitch}
  ) {constA=1}
  MaterialFunctionCall(Emissive=@n1.RGB) {materialFunction=&MF_FX_EmissiveAdjustments}
 ]
}
