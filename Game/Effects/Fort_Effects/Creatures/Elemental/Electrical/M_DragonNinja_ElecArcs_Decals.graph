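# Unlit translucent particle sprite material that draws flickering electric arcs for DragonNinja decal effects, tinted by particle color.
# Time-stepped panning noise distorts a scrolling lightning gradient texture, whose channels drive a cycling sphere mask and HeightLerp reveal.
# Opacity combines a radial mask, squared fade by distance between particle and scene behind it, and optional layered-sine pulse.
# Emissive gets an HDR remap, and world position offset pulls sprites toward the camera by half their size.
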Material /Game/Effects/Fort_Effects/Creatures/Elemental/Electrical/M_DragonNinja_ElecArcs_Decals
graph {
 properties {
  bEnableResponsiveAA=true
  bUsedWithParticleSprites=true
  blendMode=BLEND_Translucent
  pixelDepthOffsetMode=PDOM_Legacy
  shadingModel=MSM_Unlit
  usedShadingModels=Unlit
 }
 nodes [
  n1 = Multiply(A=Time()) {constB=16}
  n2 = Panner(Coordinate=TextureCoordinate() {uTiling=0.46, vTiling=0.2476}, Time=Divide(A=Ceil(None=@n1)) {constB=4}) {speedX=0.4, speedY=0.4}
  n3 = Subtract(A=ComponentMask(None=TextureSample(UVs=@n2) {samplerType=SAMPLERTYPE_LinearColor, texture=/Game/Global/Textures/Default/Noise/TilingNoise06_N_uncompressed}.RGB) {g=true, r=true}) {constB=0.5}
  n4 = TextureCoordinate()
  n5 = Time()
  n6 = TextureSample(UVs=Add(A=Multiply(A=@n3) {constB=0.25}, B=Panner(Coordinate=@n4, Time=@n5) {speedX=-0.2301, speedY=0.0354}) {constB=0}) {samplerType=SAMPLERTYPE_LinearColor, texture=/Game/Global/Textures/Default/Masks_and_Gradients/TestLightningGradient}
  n7 = Frac(None=Add(A=Time(), B=@n6.B))
  n8 = SphereMask(A=@n6.G, B=Multiply(A=Subtract(A=@n7) {constB=0.2}) {constB=10}) {attenuationRadius=1, desc="! 50 perc", hardnessPercent=57.9646}
  n9 = OneMinus(None=Clamp(None=Multiply(A=@n8) {constB=20}))
  n10 = OneMinus(None=@n6.A)
  n11 = ParticleColor()
  n12 = Multiply(
   A=Multiply(
    A=@n6.R
    B=MaterialFunctionCall(Transition Phase=@n9, Height Texture=@n10) {materialFunction=&HeightLerp}.Results
   )
   B=@n11.RGB
  )
  n13 = Constant() {r=0.5}
  n14 = Multiply(A=Multiply(A=@n12, B=@n11.A), B=SphereMask(A=TextureCoordinate(), B=@n13, Radius=@n13) {hardnessPercent=0})
  n15 = Divide(
   A=Distance(
    A=MaterialFunctionCall() {materialFunction=&WorldPositionBehindTranslucency}.Result
    B=ParticlePositionWS()
   )
   B=ScalarParameter() {defaultValue=5, group=Distance Fade, parameterName=Fade Distance}
  )
  n16 = ScalarParameter() {defaultValue=5, group=Distance Fade, parameterName=Fade Start Distance, sortPriority=1}
  n17 = OneMinus(None=Clamp(None=Subtract(A=@n15, B=@n16)))
  n18 = Multiply(A=@n17, B=@n17)
  n19 = Time() {bOverride_Period=true, period=2.6}
  n20 = LinearInterpolate(Alpha=ScalarParameter() {defaultValue=1, parameterName=RandomOffsetPerClient_Mesh})
  n21 = Time() {bOverride_Period=true, period=1.57}
  n22 = Add(A=Time() {bOverride_Period=true, period=1.24}, B=@n20)
  n23 = LinearInterpolate(
   A=Sine(None=Add(A=@n19, B=@n20)) {period=2.6}
   B=Sine(None=Add(A=@n21, B=@n20)) {period=1.57}
   Alpha=LinearInterpolate(A=Constant() {r=0.15}, B=Constant() {r=0.85}, Alpha=Sine(None=@n22) {period=1.24}) {constA=0.15, constB=0.85}
  )
  n24 = Constant() {r=1}
  Clamp(None=Multiply(A=Multiply(A=@n14, B=@n18), B=StaticSwitchParameter(True=@n23, False=@n24) {defaultValue=true, parameterName=UsePulse?}))
  MaterialFunctionCall(Emissive=MaterialFunctionCall(LDR Input=Clamp(None=@n12), HDR Tint=@n11.RGB) {materialFunction=&DeriveHDRfromLDR}.RemappedRange) {materialFunction=&MF_FX_EmissiveAdjustments}
  n25 = ParticlePositionWS()
  n26 = CameraPositionWS()
  n27 = ParticleSize()
  Multiply(A=Normalize(VectorInput=Subtract(A=@n25, B=@n26)), B=Multiply(A=ComponentMask(None=@n27) {r=true}) {constB=-0.5}) {desc=WPO input}
 ]
}
